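# Corrective Control Rig for the RavenQuill_02 Quill body that adjusts collar and back skirt bones to follow the pose.
# Head rotation and a right-upperarm spherical pose reader weight collar offsets; the reader also sets an inverted UpperArm_R_Up curve.
# Pelvis height blends crouch/standing weights, so calf bends move the opposite skirt when crouched, the same-side one Kalman-smoothed when standing.
# A left-collar-to-neck parent constraint is left unconnected.
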ControlRigBlueprint /Game/Characters/Player/Male/Medium/Bodies/M_MED_Quill/Meshes/RavenQuill_02_Rigids/M_MED_Quill_02_CtrlRig
graphs [
 RigVMFunctionLibrary {links [], nodes []}
 RigVMModel {nodes [
   RigUnit_BeginExecution = &RigUnit_BeginExecution()
   RigUnit_GetRelativeTransformForItem = &RigUnit_GetRelativeTransformForItem(Child=(Type=Bone,Name="pelvis"), Parent=(Type=Bone,Name="root"), bChildInitial=False, bParentInitial=False)
   n1 = &RigVMFunction_MathDoubleRemap(
    SourceMaximum=75
    SourceMinimum=80
    TargetMaximum=1
    TargetMinimum=0
    Value=@RigUnit_GetRelativeTransformForItem.RelativeTransform.Translation.Z
   )
   RigVMVariableNode(Value=@n1.Result, ExecuteContext=@RigUnit_BeginExecution.ExecutePin) {variable=Is crouch}
   > n2 = RigVMVariableNode(Value=&RigVMFunction_MathDoubleRemap(SourceMaximum=1, SourceMinimum=0, TargetMaximum=0, TargetMinimum=1, Value=@n1.Result).Result) {variable=Isnt Crouch}
   RigVMFunction_Sequence = &RigVMFunction_Sequence(ExecuteContext=@n2.ExecuteContext)
   n3 = &RigVMFunction_Sequence(ExecuteContext=@RigVMFunction_Sequence.A)
   RigVMFunction_MathQuaternionToEuler = &RigVMFunction_MathQuaternionToEuler(
    RotationOrder=ZYX
    Value=&RigUnit_GetRelativeTransformForItem(Child=(Type=Bone,Name="head"), Parent=(Type=Bone,Name="spine_05"), bChildInitial=False, bParentInitial=False).RelativeTransform.Rotation
   )
   n4 = &RigVMFunction_MathDoubleRemap(SourceMaximum=-50, SourceMinimum=0, TargetMaximum=1, TargetMinimum=0, Value=@RigVMFunction_MathQuaternionToEuler.Result.Z)
   &RigUnit_ModifyTransforms(
    ItemToModify=((Item=(Type=Bone,Name="Collar_M_Base"),Transform=(Rotation=(X=0.083,Y=0.0479,Z=0.4977,W=0.862),Translation=(X=-2,Y=0,Z=0),Scale3D=(X=1,Y=1,Z=1))))
    WeightMinimum=0
    Weight=@n4.Result
    ExecutePin=@n3.A
   )
   > &RigUnit_ModifyTransforms(
    ItemToModify="((Item=(Type=Bone,Name=\"Collar_M_Base\"),Transform=(Rotation=(X=-0.069,Y=0.0103,Z=-0.1474,W=0.9866),Translation=(X=0,Y=3,Z=0),Scale3D=(X=1,Y=1,Z=1))),(Item=(Type=Bone,Name=\"Collar_R_Base\"),Transform=(Rotation=(X=-0.2239,Y=-0.3561,Z=0.0916,W=0.9026),Translat ...(truncated)"
    WeightMinimum=0
    Weight=&RigVMFunction_MathDoubleRemap(SourceMaximum=-75, SourceMinimum=-50, TargetMaximum=1, TargetMinimum=0, Value=@RigVMFunction_MathQuaternionToEuler.Result.Z).Result
   )
   > &RigUnit_ModifyTransforms(
    ItemToModify="((Item=(Type=Bone,Name=\"Collar_M_Base\"),Transform=(Rotation=(X=0.1107,Y=-0.0235,Z=-0.2066,W=0.9719),Translation=(X=0,Y=2,Z=0),Scale3D=(X=1,Y=1,Z=1))),(Item=(Type=Bone,Name=\"Collar_R_Base\"),Transform=(Rotation=(X=-0.3732,Y=-0.0326,Z=0.0808,W=0.9237),Transla ...(truncated)"
    WeightMinimum=0
    Weight=&RigVMFunction_MathDoubleMul(
     B=@n4.Result
     A=&RigVMFunction_MathDoubleRemap(SourceMaximum=-49, SourceMinimum=-30, TargetMaximum=1, TargetMinimum=0, Value=@RigVMFunction_MathQuaternionToEuler.Result.X).Result
    ).Result
   )
   n5 = &RigUnit_SphericalPoseReader(
    ActiveRegionScaleFactors=(PositiveWidth=1,NegativeWidth=1.3,PositiveHeight=1,NegativeHeight=1)
    ActiveRegionSize=0.1
    Debug=(bDrawDebug=True,bDraw2D=False,bDrawLocalAxes=False,DebugScale=25,DebugSegments=20,DebugThickness=0.25)
    DriverAxis=(X=-1,Y=0,Z=0)
    DriverItem=(Type=Bone,Name="upperarm_r")
    FalloffRegionScaleFactors=(PositiveWidth=1,NegativeWidth=2,PositiveHeight=1,NegativeHeight=1)
    FalloffSize=0.2
    FlipHeightScaling=False
    FlipWidthScaling=False
    OptionalParentItem=(Type=Bone,Name="spine_05")
    RotationOffset=(X=-80,Y=-10,Z=40)
    ExecutePin=@n3.B
   )
   > /Script/ControlRig.RigUnit_SetCurveValue(
    Curve=UpperArm_R_Up
    Value=&RigVMFunction_MathDoubleRemap(SourceMaximum=1, SourceMinimum=0, TargetMaximum=0, TargetMinimum=1, Value=@n5.OutputParam).Result
   ) {is=struct RigVMUnitNode}
   > &RigUnit_ModifyTransforms(
    ItemToModify="((Item=(Type=Bone,Name=\"Collar_R_Base\"),Transform=(Rotation=(X=-0.1695,Y=0.0376,Z=-0.2132,W=0.9615),Translation=(X=0,Y=0,Z=0),Scale3D=(X=1,Y=1,Z=1))),(Item=(Type=Bone,Name=\"Collar_R_03\"),Transform=(Rotation=(X=0,Y=0,Z=-0.342,W=0.9397),Translation=(X=0,Y=0, ...(truncated)"
    WeightMinimum=0
    Weight=@n5.OutputParam
   )
   n6 = &RigVMFunction_Sequence(ExecuteContext=@RigVMFunction_Sequence.B)
   Get Delta Ref = RigVMFunctionReferenceNode(Item=(Type=Bone,Name="calf_l"), ExecuteContext=@n6.A)
   Remap = &RigVMFunction_MathDoubleRemap(SourceMaximum=115, SourceMinimum=50, TargetMaximum=1, TargetMinimum=0, Value=@"Get Delta Ref".Result.Rotation.Yaw)
   n7 = RigVMVariableNode() {variable=Is crouch}
   Multiply = &RigVMFunction_MathDoubleMul(A=@Remap.Result, B=@n7.Value)
   > &RigUnit_ModifyTransforms(
    ItemToModify=((Item=(Type=Bone,Name="dyn_skirt_r_back_base"),Transform=(Rotation=(X=0.4226,Y=0,Z=0,W=0.9063),Translation=(X=-5,Y=0,Z=-7),Scale3D=(X=1,Y=1,Z=1))))
    WeightMinimum=0
    Weight=@Multiply.Result
   )
   > &RigUnit_ModifyTransforms(
    ItemToModify=((Item=(Type=Bone,Name="dyn_skirt_r_back_base"),Transform=(Rotation=(X=0.0872,Y=0,Z=0,W=0.9962),Translation=(X=-5,Y=2,Z=-7),Scale3D=(X=1,Y=1,Z=1))))
    WeightMinimum=0
    Weight=&RigVMFunction_MathDoubleMul(
     A=&RigVMFunction_MathDoubleRemap(SourceMaximum=125, SourceMinimum=115, TargetMaximum=1, TargetMinimum=0, Value=@"Get Delta Ref".Result.Rotation.Yaw).Result
     B=@n7.Value
    ).Result
   )
   KalmanFilter = &RigVMFunction_KalmanFloat(
    BufferSize=8
    Value=&RigVMFunction_MathDoubleMul(
     A=&RigVMFunction_MathDoubleRemap(SourceMaximum=120, SourceMinimum=60, TargetMaximum=1, TargetMinimum=0, Value=@"Get Delta Ref".Result.Rotation.Yaw).Result
     B=RigVMVariableNode() {variable=Isnt Crouch}.Value
    ).Result
   ) {template="KalmanFilter::Execute(in Value,in BufferSize,out Result)"}
   > &RigUnit_ModifyTransforms(
    ItemToModify=((Item=(Type=Bone,Name="dyn_skirt_l_back_base"),Transform=(Rotation=(X=-0.5714,Y=0.0714,Z=0.05,W=0.816),Translation=(X=-15,Y=-10,Z=-3),Scale3D=(X=1,Y=1,Z=1))))
    WeightMinimum=0
    Weight=@KalmanFilter.Result
   )
   Get Delta Ref_1 = RigVMFunctionReferenceNode(Item=(Type=Bone,Name="calf_r"), ExecuteContext=@n6.B)
   > &RigUnit_ModifyTransforms(
    ItemToModify=((Item=(Type=Bone,Name="dyn_skirt_l_back_base"),Transform=(Rotation=(X=0,Y=0.2588,Z=0,W=0.9659),Translation=(X=0,Y=-5,Z=-5),Scale3D=(X=1,Y=1,Z=1))))
    WeightMinimum=0
    Weight=&RigVMFunction_MathDoubleMul(
     A=&RigVMFunction_MathDoubleRemap(SourceMaximum=115, SourceMinimum=50, TargetMaximum=1, TargetMinimum=0, Value=@"Get Delta Ref_1".Result.Rotation.Yaw).Result
     B=RigVMVariableNode() {variable=Is crouch}.Value
    ).Result
   )
   n8 = &RigVMFunction_MathDoubleRemap(SourceMaximum=120, SourceMinimum=60, TargetMaximum=1, TargetMinimum=0, Value=@"Get Delta Ref_1".Result.Rotation.Yaw)
   n9 = RigVMVariableNode() {variable=Isnt Crouch}
   > &RigUnit_ModifyTransforms(
    ItemToModify=((Item=(Type=Bone,Name="dyn_skirt_r_back_base"),Transform=(Rotation=(X=0.5649,Y=0.1422,Z=-0.0996,W=0.8067),Translation=(X=-5,Y=10,Z=-5),Scale3D=(X=1,Y=1,Z=1))))
    WeightMinimum=0
    Weight=&RigVMFunction_KalmanFloat(BufferSize=8, Value=&RigVMFunction_MathDoubleMul(A=@n8.Result, B=@n9.Value).Result) {template="KalmanFilter::Execute(in Value,in BufferSize,out Result)"}.Result
   )
   &RigUnit_ParentConstraint(
    AdvancedSettings=(InterpolationType=Average,RotationOrderForFilter=XZY)
    Child=(Type=Bone,Name="Collar_L_Base")
    ExecutePin=()
    Parents=((Item=(Type=Bone,Name="neck_01"),Weight=1))
    Weight=1
   )
  ]}
]
hierarchy [
 root {loc=<0 0 0>, rot=<0 0 0 1>}
 attach {parent=root, loc=<0 0 0>, rot=<0 0 0 1>}
 pelvis {parent=root, loc=<0 -0.7071 93.1191>, rot=<0.0127 0.707 -0.0127 -0.707>}
 spine_01 {parent=pelvis, loc=<6.7494 -0.6388 0>, rot=<0 0 0.034 -0.9994>, scale=<1.0092 1 1>}
 spine_02 {parent=spine_01, loc=<7.2518 -0.035 0>, rot=<0 0 0.0263 -0.9997>, scale=<0.9909 1 1>}
 spine_03 {parent=spine_02, loc=<6.0104 0.0273 0>, rot=<0 0 0.0126 -0.9999>}
 spine_04 {parent=spine_03, loc=<8.1024 0 0>, rot=<0 0 -0.0132 -0.9999>}
 spine_05 {parent=spine_04, loc=<10.0146 0 0>, rot=<0 0 -0.0015 -1>}
 clavicle_l {parent=spine_05, loc=<4.3575 -0.1466 -2.3223>, rot=<0.0465 -0.7289 -0.009 -0.6829>}
 upperarm_l {parent=clavicle_l, loc=<15.5111 0 0>, rot=<0.0305 -0.536 -0.0057 -0.8437>}
 lowerarm_l {parent=upperarm_l, loc=<26.6514 0 0>, rot=<0.0391 -0.0099 -0.2453 -0.9686>}
 hand_l {parent=lowerarm_l, loc=<22.3491 0 0>, rot=<-0.7081 -0.0507 0.0975 -0.6975>}
 index_metacarpal_l {parent=hand_l, loc=<1.6535 0.0183 -2.96>, rot=<0.0158 -0.0124 -0.0173 -0.9996>}
 index_01_l {parent=index_metacarpal_l, loc=<8.3008 0.0357 -0.0524>, rot=<-0.0044 0.0485 0.0924 -0.9945>}
 index_02_l {parent=index_01_l, loc=<3.3587 -0.3004 0.0452>, rot=<-0.002 0.0027 0.3225 -0.9466>}
 index_03_l {parent=index_02_l, loc=<2.5577 -0.1233 -0.0252>, rot=<-0.002 0.0067 0.2558 -0.9667>}
 middle_metacarpal_l {parent=hand_l, loc=<2.1856 -0.1725 -0.5984>, rot=<0.006 0.0168 -0.0329 -0.9993>}
 middle_01_l {parent=middle_metacarpal_l, loc=<7.3275 -0.2447 0.0129>, rot=<-0.0208 0.0042 0.0858 -0.9961>}
 middle_02_l {parent=middle_01_l, loc=<4.0131 0.0722 0.0029>, rot=<-0.0218 0.0258 0.3512 -0.9357>}
 middle_03_l {parent=middle_02_l, loc=<2.6079 0.1152 0.0049>, rot=<0.0082 0.0188 0.1708 -0.9851>}
 pinky_metacarpal_l {parent=hand_l, loc=<1.7616 -0.9753 3.3814>, rot=<-0.0397 0.0814 -0.069 -0.9935>}
 pinky_01_l {parent=pinky_metacarpal_l, loc=<7.1709 -0.4831 -0.1094>, rot=<-0.0654 -0.0035 0.2242 -0.9723>}
 pinky_02_l {parent=pinky_01_l, loc=<2.8274 0.0893 -0.1043>, rot=<0.0037 0.0011 0.2854 -0.9584>}
 pinky_03_l {parent=pinky_02_l, loc=<2.2046 0.0307 -0.0078>, rot=<-0.0002 0.0098 0.1445 -0.9895>}
 ring_metacarpal_l {parent=hand_l, loc=<2.2074 -0.3625 1.3638>, rot=<-0.0284 0.0722 -0.0458 -0.9959>}
 ring_01_l {parent=ring_metacarpal_l, loc=<7.0295 -0.6197 -0.1446>, rot=<-0.0489 -0.0252 0.1892 -0.9804>}
 ring_02_l {parent=ring_01_l, loc=<3.6174 0.4647 -0.1027>, rot=<0 0 0.2806 -0.9598>}
 ring_03_l {parent=ring_02_l, loc=<2.4152 0.114 0.0199>, rot=<0 0 0.1189 -0.9929>}
 thumb_01_l {parent=hand_l, loc=<1.3118 -0.3191 -2.459>, rot=<0.6546 -0.4042 0.0664 -0.6354>}
 thumb_02_l {parent=thumb_01_l, loc=<5.9336 -0.0448 0.0401>, rot=<-0.0037 0.0213 0.298 -0.9543>}
 thumb_03_l {parent=thumb_02_l, loc=<2.2223 0.3682 -0.0907>, rot=<0.0118 -0.013 0.0066 -0.9998>}
 weapon_l {parent=hand_l, loc=<-0.1974 -3.4111 0.3811>, rot=<0 0 0 1>}
 lowerarm_twist_01_l {parent=lowerarm_l, loc=<17.8549 0 0>, rot=<0 0 0 1>}
 lowerarm_twist_02_l {parent=lowerarm_l, loc=<9.6199 0 0>, rot=<0 0 0 1>}
 upperarm_twist_01_l {parent=upperarm_l, loc=<0.0041 0 0>, rot=<0 0 0 1>}
 upperarm_twist_02_l {parent=upperarm_l, loc=<13.7941 0 0>, rot=<0 0 0 1>}
 dyn_shoulder_l {parent=upperarm_l, loc=<0 0 0>, rot=<0 0 0 1>}
 clavicle_r {parent=spine_05, loc=<4.3575 -0.1466 2.3223>, rot=<0.7289 0.0465 -0.6829 0.009>}
 upperarm_r {parent=clavicle_r, loc=<-15.5111 0 0>, rot=<0.0305 -0.536 -0.0057 -0.8437>}
 lowerarm_r {parent=upperarm_r, loc=<-26.6515 0 0>, rot=<0.0391 -0.0099 -0.2453 -0.9686>}
 hand_r {parent=lowerarm_r, loc=<-22.3492 0 0>, rot=<-0.7081 -0.0507 0.0975 -0.6975>}
 index_metacarpal_r {parent=hand_r, loc=<-1.6534 -0.0181 2.9601>, rot=<0.0158 -0.0124 -0.0173 -0.9996>}
 index_01_r {parent=index_metacarpal_r, loc=<-8.3008 -0.0357 0.0524>, rot=<-0.0044 0.0485 0.0924 -0.9945>}
 index_02_r {parent=index_01_r, loc=<-3.3589 0.3003 -0.0453>, rot=<-0.002 0.0027 0.3225 -0.9466>}
 index_03_r {parent=index_02_r, loc=<-2.5578 0.1231 0.0251>, rot=<-0.002 0.0067 0.2558 -0.9667>}
 middle_metacarpal_r {parent=hand_r, loc=<-2.1856 0.1726 0.5984>, rot=<0.006 0.0168 -0.0329 -0.9993>}
 middle_01_r {parent=middle_metacarpal_r, loc=<-7.3273 0.2447 -0.0129>, rot=<-0.0208 0.0042 0.0858 -0.9961>}
 middle_02_r {parent=middle_01_r, loc=<-4.0142 -0.1411 -0.07>, rot=<-0.0216 0.026 0.3512 -0.9357>}
 middle_03_r {parent=middle_02_r, loc=<-2.6074 -0.0534 -0.1312>, rot=<0.0084 0.0187 0.1708 -0.9851>}
 pinky_metacarpal_r {parent=hand_r, loc=<-1.7617 0.9753 -3.3814>, rot=<-0.0397 0.0814 -0.069 -0.9935>}
 pinky_01_r {parent=pinky_metacarpal_r, loc=<-7.1707 0.4832 0.1094>, rot=<-0.0654 -0.0035 0.2242 -0.9723>}
 pinky_02_r {parent=pinky_01_r, loc=<-2.8274 -0.0892 0.1043>, rot=<0.0037 0.0011 0.2854 -0.9584>}
 pinky_03_r {parent=pinky_02_r, loc=<-2.2045 -0.031 0.0078>, rot=<-0.0002 0.0098 0.1445 -0.9895>}
 ring_metacarpal_r {parent=hand_r, loc=<-2.2073 0.3627 -1.3638>, rot=<-0.0284 0.0722 -0.0458 -0.9959>}
 ring_01_r {parent=ring_metacarpal_r, loc=<-7.0297 0.6197 0.1446>, rot=<-0.0489 -0.0252 0.1892 -0.9804>}
 ring_02_r {parent=ring_01_r, loc=<-3.6171 -0.4648 0.1027>, rot=<0 0 0.2806 -0.9598>}
 ring_03_r {parent=ring_02_r, loc=<-2.4153 -0.1141 -0.0199>, rot=<0 0 0.1189 -0.9929>}
 thumb_01_r {parent=hand_r, loc=<-1.3119 0.3192 2.459>, rot=<0.6546 -0.4042 0.0664 -0.6354>}
 thumb_02_r {parent=thumb_01_r, loc=<-5.9335 0.0447 -0.0401>, rot=<-0.0037 0.0213 0.298 -0.9543>}
 thumb_03_r {parent=thumb_02_r, loc=<-2.2223 -0.3682 0.0907>, rot=<0.0118 -0.013 0.0066 -0.9998>}
 weapon_r {parent=hand_r, loc=<0.1975 3.4112 -0.3811>, rot=<0 0 0 -1>}
 lowerarm_twist_01_r {parent=lowerarm_r, loc=<-17.855 0 0>, rot=<0 0 0 1>}
 lowerarm_twist_02_r {parent=lowerarm_r, loc=<-9.6199 0 0>, rot=<0 0 0 1>}
 upperarm_twist_01_r {parent=upperarm_r, loc=<-0.0041 0 0>, rot=<0 0 0 1>}
 upperarm_twist_02_r {parent=upperarm_r, loc=<-13.7942 0 0>, rot=<0 0 0 1>}
 dyn_armband_knot {parent=upperarm_twist_02_r, loc=<-3.147 6.7155 -0.0308>, rot=<0.8489 0.0638 0.5203 0.068>}
 dyn_armband_hi {parent=dyn_armband_knot, loc=<1.1632 0 0>, rot=<0 0 -0.1503 -0.9886>}
 dyn_armband_low {parent=dyn_armband_knot, loc=<0.1966 0.1373 -0.9804>, rot=<0.03 0.0034 0.0773 -0.9966>}
 dyn_shoulder_r {parent=upperarm_r, loc=<0 0 0>, rot=<0 0 0 1>}
 neck_01 {parent=spine_05, loc=<11.3138 0.1529 0>, rot=<0 0 -0.2003 -0.9797>}
 neck_02 {parent=neck_01, loc=<4.5627 0 0>, rot=<0 0 0.0504 -0.9987>}
 head {parent=neck_02, loc=<5.3914 0 0>, rot=<0 0 0.115 -0.9934>}
 dyn_feathers_01 {parent=head, loc=<0 0 0>, rot=<0 0 0 1>}
 dyn_feathers_02 {parent=head, loc=<0 0 0>, rot=<0 0 0 1>}
 dyn_feathers_03 {parent=head, loc=<0 0 0>, rot=<0 0 0 1>}
 Hood_Corrective {parent=head, loc=<6.5 -10 0>, rot=<0.0029 -0.7071 -0.0029 -0.7071>}
 dyn_scarf_f {parent=spine_05, loc=<7.1505 13.2293 -0.5943>, rot=<0.0105 -0.707 -0.0105 -0.707>}
 dyn_scarf_r {parent=spine_05, loc=<14.3878 1.898 7.632>, rot=<0.0105 -0.707 -0.0105 -0.707>}
 dyn_scarf_l {parent=spine_05, loc=<15.1146 2.4248 -9.1274>, rot=<0.0105 -0.707 -0.0105 -0.707>}
 dyn_scarf_b {parent=spine_05, loc=<17.5237 -9.2448 -1.6328>, rot=<0.0105 -0.707 -0.0105 -0.707>}
 Collar_M_Base {parent=spine_05, loc=<15.3746 -8.2485 0.4845>, rot=<0.9886 -0.1386 0.0127 0.0576>}
 Collar_M_01 {parent=Collar_M_Base, loc=<0 0 0>, rot=<0 0 0 1>}
 Collar_M_02 {parent=Collar_M_01, loc=<4 0 0>, rot=<0 0 0 1>}
 Collar_M_03 {parent=Collar_M_01, loc=<4 0 -4.2984>, rot=<0 0 0 1>}
 Collar_M_04 {parent=Collar_M_01, loc=<4 0 4.8152>, rot=<0 0 0 1>}
 Collar_R_Base {parent=spine_05, loc=<10 5 9.3>, rot=<-0.6698 -0.0472 -0.3371 -0.6599>}
 Collar_R_01 {parent=Collar_R_Base, loc=<0 0 0>, rot=<0 0 0 1>}
 Collar_R_02 {parent=Collar_R_01, loc=<4 0 0>, rot=<0 0 0 1>}
 Collar_R_03 {parent=Collar_R_01, loc=<4 0 -3.951>, rot=<0 0 0 1>}
 Collar_R_04 {parent=Collar_R_01, loc=<4 0 5.1186>, rot=<0 0 0 1>}
 Collar_L_Base {parent=spine_05, loc=<10 5 -9.3>, rot=<-0.6726 -0.3311 -0.0544 -0.6596>}
 Collar_L_01 {parent=Collar_L_Base, loc=<0 0 0>, rot=<0 0 0 1>}
 Collar_L_02 {parent=Collar_L_01, loc=<4 0 0>, rot=<0 0 0 1>}
 Collar_L_03 {parent=Collar_L_01, loc=<4 0 6.6604>, rot=<0 0 0 1>}
 Collar_L_04 {parent=Collar_L_01, loc=<4 0 -4.1917>, rot=<0 0 0 1>}
 dyn_tags {parent=spine_04, loc=<2.458 15.4655 -1.1202>, rot=<0.0584 -0.7047 -0.0584 -0.7047>}
 thigh_l {parent=pelvis, loc=<-2.5693 -0.0109 -7.7501>, rot=<0.0012 0.0572 0.0155 -0.9982>}
 calf_l {parent=thigh_l, loc=<-39.0484 0 0>, rot=<0.0006 -0.0161 -0.0391 -0.9991>}
 calf_twist_01_l {parent=calf_l, loc=<-35.813 -0.0967 0.0397>, rot=<-0.0573 0.0092 -0.0043 -0.9983>}
 calf_twist_02_l {parent=calf_l, loc=<-18.8064 -0.0967 0.0397>, rot=<-0.0168 0.0094 -0.0039 -0.9998>}
 foot_l {parent=calf_l, loc=<-42.561 0.1077 0.4292>, rot=<0.0416 -0.0446 0.0398 -0.9973>}
 ball_l {parent=foot_l, loc=<-6.1599 13.8264 0.6806>, rot=<0.0001 -0.0001 -0.7063 -0.7079>}
 thigh_twist_01_l {parent=thigh_l, loc=<-21.6303 0 0>, rot=<-0.0059 0 0 -1>}
 dyn_skirt_l_front_base {parent=thigh_twist_01_l, loc=<-3.5576 5.8988 -10.7672>, rot=<-0.0344 -0.3346 0.9415 0.0216>}
 dyn_skirt_l_front_parent {parent=dyn_skirt_l_front_base, loc=<0 0 0>, rot=<0 0 0 1>}
 dyn_skirt_l_front_01 {parent=dyn_skirt_l_front_parent, loc=<12.1362 1.594 1.6946>, rot=<-0.0864 0.0447 0.0042 -0.9952>}
 dyn_skirt_l_front_02 {parent=dyn_skirt_l_front_01, loc=<15.8467 0.1254 0.4066>, rot=<0.0003 0.0108 -0.0025 -0.9999>}
 dyn_skirt_l_back_base {parent=thigh_twist_01_l, loc=<-3.4986 -12.126 -5.9665>, rot=<0.0357 0.5046 0.8624 -0.0176>}
 dyn_skirt_l_back_parent {parent=dyn_skirt_l_back_base, loc=<0 0 0>, rot=<0 0 0 1>}
 dyn_skirt_l_back_01 {parent=dyn_skirt_l_back_parent, loc=<11.9565 -1.3911 0.5783>, rot=<-0.0177 -0.0199 0.0672 -0.9974>}
 dyn_skirt_l_back_02 {parent=dyn_skirt_l_back_01, loc=<18.2218 -3.0868 1.5978>, rot=<0.1743 0.0238 -0.0739 -0.9816>}
 dyn_feather_side_l_01 {parent=thigh_l, loc=<-4.9442 -5.0124 -11.0335>, rot=<-0.1997 0.1653 0.9644 -0.0521>}
 dyn_feather_side_l_02 {parent=thigh_l, loc=<-4.4728 0.3948 -13.136>, rot=<-0.2032 0.0004 0.9791 0.0024>}
 thigh_r {parent=pelvis, loc=<-2.5691 -0.0109 7.7501>, rot=<0.0572 -0.0012 0.9982 0.0155>}
 calf_r {parent=thigh_r, loc=<39.0487 0 0>, rot=<0.0006 -0.0161 -0.0391 -0.9991>}
 calf_twist_01_r {parent=calf_r, loc=<35.8132 0.0967 -0.0397>, rot=<-0.3532 0.0075 -0.0068 -0.9355>}
 calf_twist_02_r {parent=calf_r, loc=<18.8064 0.0967 -0.0397>, rot=<-0.1679 0.0087 -0.0053 -0.9858>}
 foot_r {parent=calf_r, loc=<42.5612 -0.1077 -0.4292>, rot=<0.0416 -0.0446 0.0398 -0.9973>}
 ball_r {parent=foot_r, loc=<6.1598 -13.8264 -0.6806>, rot=<0.0001 -0.0001 -0.7063 -0.7079>}
 thigh_twist_01_r {parent=thigh_r, loc=<21.6303 0 0>, rot=<0.027 0 0 -0.9996>}
 dyn_skirt_r_back_base {parent=thigh_twist_01_r, loc=<3.4164 10.9405 7.2087>, rot=<0.9304 0.0533 0.0282 0.3617>}
 dyn_skirt_r_back_base_parent {parent=dyn_skirt_r_back_base, loc=<0 0 0>, rot=<0 0 0 1>}
 dyn_skirt_r_back_01 {parent=dyn_skirt_r_back_base_parent, loc=<12.3433 3.9788 2.3606>, rot=<-0.1869 0.0031 -0.0448 -0.9814>}
 dyn_skirt_r_back_02 {parent=dyn_skirt_r_back_01, loc=<17.4845 5.4672 0.8912>, rot=<-0.1152 0.0395 -0.0713 -0.99>}
 dyn_skirt_r_front_base {parent=thigh_twist_01_r, loc=<3.6403 -6.0452 10.4487>, rot=<0.9296 -0.0523 0.0277 -0.3637>}
 dyn_skirt_r_front_parent {parent=dyn_skirt_r_front_base, loc=<0 0 0>, rot=<0 0 0 1>}
 dyn_skirt_r_front_01 {parent=dyn_skirt_r_front_parent, loc=<12.4151 0.1109 0.7427>, rot=<0.0858 0 0.0151 -0.9962>}
 dyn_skirt_r_front_02 {parent=dyn_skirt_r_front_01, loc=<15.8101 0.3601 3.4059>, rot=<-0.019 0.0383 0.0172 -0.9989>}
 dyn_feather_side_r_01 {parent=thigh_r, loc=<6.7301 6.2055 11.5256>, rot=<0.9613 0.0575 0.199 0.1816>}
 dyn_pouch_r {parent=pelvis, loc=<0 0 0>, rot=<0.7071 -0.0013 -0.7071 -0.0013>}
 dyn_skirtFeathers {parent=pelvis, loc=<0 0 0>, rot=<0.7071 -0.0013 -0.7071 -0.0013>}
 dyn_grenade01 {parent=pelvis, loc=<10.3879 -11.1878 -6.456>, rot=<0.3855 -0.5928 0.3877 -0.5914>}
 dyn_grenade02 {parent=pelvis, loc=<11.5638 -4.6352 -13.3803>, rot=<0.1226 -0.6964 0.1251 -0.6959>}
 ik_foot_root {parent=root, loc=<0 0 0>, rot=<0 0 0 1>}
 ik_foot_l {parent=ik_foot_root, loc=<15.2608 -4.4176 9.4433>, rot=<0.0315 0.7052 0.0319 -0.7076>}
 ik_foot_r {parent=ik_foot_root, loc=<-15.2609 -4.4176 9.4431>, rot=<0 0 0 1>}
 ik_hand_root {parent=root, loc=<0 0 0>, rot=<0 0 0 1>}
 ik_hand_gun {parent=ik_hand_root, loc=<-33.8029 3.8261 90.6359>, rot=<0.4834 0.3223 0.5887 -0.5621>}
 ik_hand_l {parent=ik_hand_gun, loc=<6.7067 65.8041 13.978>, rot=<-0.0992 -0.9733 -0.2068 0>}
 ik_hand_r {parent=ik_hand_gun, loc=<0 0 0>, rot=<0 0 0 1>}
 UpperArm_R_Up {type=Curve, loc=<0 0 0>, rot=<0 0 0 1>}
]
variables [
 Is crouch {type=double, default=0}
 Isnt Crouch {type=double, default=0}
]
