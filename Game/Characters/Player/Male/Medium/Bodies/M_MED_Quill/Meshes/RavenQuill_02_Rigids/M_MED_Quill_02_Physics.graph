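# Physics asset for the RavenQuill_02 rigid meshes of the medium male Quill player body, covering dynamic skirt motion around the legs.
# Four box-body chains (left and right, front and back) run from a small base through three segments with limited swing and twist.
# On each side, a free-rotation constraint ties the front and back tips together.
# Unconstrained capsules on thighs and calves provide leg collision shapes for the skirt.
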PhysicsAsset /Game/Characters/Player/Male/Medium/Bodies/M_MED_Quill/Meshes/RavenQuill_02_Rigids/M_MED_Quill_02_Physics
bodies {
 calf_l [
  {shapeType=Capsule, length=22.5648, radius=10.4746, center=<-29.1882 -0.0004 4.4517>, rotation=<rot 84.389 0 0>}
  {shapeType=Capsule, length=9.0963, radius=10.9037, center=<-15.789 0.2692 1.3906>, rotation=<rot 86.9738 -179.9999 -171.1217>}
 ]
 calf_r [
  {shapeType=Capsule, length=25.7765, radius=8.25, center=<27.1269 -0.0002 0.3787>, rotation=<rot 87.0306 -179.9999 -180>}
  {shapeType=Capsule, length=23.5, radius=8.5, center=<11.4995 1.981 -2.6688>, rotation=<rot 89.4518 0.0002 0.0002>}
 ]
 dyn_skirt_l_back_01 [{shapeType=Box, extentX=17.6164, extentY=16.5805, extentZ=4.5492, center=<9.1202 0 0>, rotation=<rot 0 0 3.7866>}]
 dyn_skirt_l_back_02 [{shapeType=Box, extentX=19.8874, extentY=13.7764, extentZ=4.7276, center=<9.4909 0 0>}]
 dyn_skirt_l_back_base [{shapeType=Box, extentX=2, extentY=2, extentZ=2}]
 dyn_skirt_l_back_parent [{shapeType=Box, extentX=15.2568, extentY=20, extentZ=4.5964, center=<6.1985 0 0>}]
 dyn_skirt_l_front_01 [{shapeType=Box, extentX=16.6164, extentY=20, extentZ=4.5964, center=<7.9777 0 0.0002>}]
 dyn_skirt_l_front_02 [{shapeType=Box, extentX=17.1164, extentY=20, extentZ=4.5964, center=<8.7409 0 0.0003>}]
 dyn_skirt_l_front_base [{shapeType=Box, extentX=2, extentY=2, extentZ=2}]
 dyn_skirt_l_front_parent [{shapeType=Box, extentX=12.6164, extentY=20, extentZ=4.5964, center=<6.1521 0 0.0002>}]
 dyn_skirt_r_back_01 [{shapeType=Box, extentX=16.6494, extentY=12.809, extentZ=4.9398, center=<8.0495 0 -0.0002>}]
 dyn_skirt_r_back_02 [{shapeType=Box, extentX=15.4257, extentY=12.198, extentZ=4.6902, center=<7.4535 0 0>}]
 dyn_skirt_r_back_base [{shapeType=Box, extentX=1, extentY=1, extentZ=1}]
 dyn_skirt_r_back_base_parent [{shapeType=Box, extentX=18.3433, extentY=15.634, extentZ=4.0964, center=<4.9686 0 0>}]
 dyn_skirt_r_front_01 [{shapeType=Box, extentX=15.8398, extentY=14.3983, extentZ=4.8234, center=<6.6818 0.0001 1.6027>, rotation=<rot 9.1421 0 -0.0001>}]
 dyn_skirt_r_front_02 [{shapeType=Box, extentX=11.735, extentY=14.1887, extentZ=4.1996, center=<4.989 0 0>}]
 dyn_skirt_r_front_base [{shapeType=Box, extentX=1, extentY=1, extentZ=1}]
 dyn_skirt_r_front_parent [{shapeType=Box, extentX=11.838, extentY=13.4216, extentZ=2.558, center=<5.1377 0 0>}]
 thigh_l [{shapeType=Capsule, length=26.0425, radius=9.9575, center=<-30.5558 -1.6894 1.7414>, rotation=<rot 87.0153 179.9998 178.9284>}]
 thigh_r [{shapeType=Capsule, length=21, radius=9, center=<26.6749 0.0003 -1.2365>, rotation=<rot 87.0252 -179.9999 -180>}]
}
constraints [
 {
  bone1Name=dyn_skirt_r_back_base_parent
  bone2Name=dyn_skirt_r_back_base
  swing1LimitDegrees=30
  swing2LimitDegrees=50
  twistLimitDegrees=28.04
 }
 {bone1Name=dyn_skirt_r_back_01, bone2Name=dyn_skirt_r_back_base_parent, swing1LimitDegrees=40, swing2LimitDegrees=40, twistLimitDegrees=23.16}
 {bone1Name=dyn_skirt_r_back_02, bone2Name=dyn_skirt_r_back_01, swing1LimitDegrees=20, swing2LimitDegrees=20, twistLimitDegrees=3}
 {bone1Name=dyn_skirt_r_front_parent, bone2Name=dyn_skirt_r_front_base, swing1LimitDegrees=30, swing2LimitDegrees=50, twistLimitDegrees=5}
 {bone1Name=dyn_skirt_r_front_01, bone2Name=dyn_skirt_r_front_parent, swing1LimitDegrees=40, swing2LimitDegrees=40, twistLimitDegrees=3}
 {bone1Name=dyn_skirt_r_front_02, bone2Name=dyn_skirt_r_front_01, swing1LimitDegrees=20, swing2LimitDegrees=20, twistLimitDegrees=3}
 {
  bone1Name=dyn_skirt_r_back_02
  bone2Name=dyn_skirt_r_front_02
  swing1LimitDegrees=45
  swing1Motion=Free
  swing2LimitDegrees=45
  swing2Motion=Free
  twistLimitDegrees=45
  twistMotion=Free
 }
 {bone1Name=dyn_skirt_l_front_parent, bone2Name=dyn_skirt_l_front_base, swing1LimitDegrees=30, swing2LimitDegrees=10, twistLimitDegrees=5}
 {bone1Name=dyn_skirt_l_front_01, bone2Name=dyn_skirt_l_front_parent, swing1LimitDegrees=10, swing2LimitDegrees=20, twistLimitDegrees=3}
 {bone1Name=dyn_skirt_l_front_02, bone2Name=dyn_skirt_l_front_01, swing1LimitDegrees=20, swing2LimitDegrees=20, twistLimitDegrees=3}
 {bone1Name=dyn_skirt_l_back_parent, bone2Name=dyn_skirt_l_back_base, swing1LimitDegrees=35, swing2LimitDegrees=50, twistLimitDegrees=20}
 {bone1Name=dyn_skirt_l_back_01, bone2Name=dyn_skirt_l_back_parent, swing1LimitDegrees=80, swing2LimitDegrees=80, twistLimitDegrees=5}
 {bone1Name=dyn_skirt_l_back_02, bone2Name=dyn_skirt_l_back_01, swing1LimitDegrees=30, swing2LimitDegrees=30, twistLimitDegrees=5}
 {
  bone1Name=dyn_skirt_l_front_02
  bone2Name=dyn_skirt_l_back_02
  swing1LimitDegrees=45
  swing1Motion=Free
  swing2LimitDegrees=45
  swing2Motion=Free
  twistLimitDegrees=45
  twistMotion=Free
 }
]
