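# Post-pose Anim Blueprint for the RavenQuill 02 rigids body of male character Quill, adding secondary motion to costume parts.
# It copies the owner's skeletal mesh pose through a Control Rig, then chains AnimDynamics on shoulders, collar, armbands and skirt feathers.
# A final RigidBody node simulates the Quill physics asset relative to root.
# The armband-knot node is unconnected, and Disabel_AnimDynamics, set each update from curve UpperArm_R_Up, drives nothing here.
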AnimBlueprint /Game/Characters/Player/Male/Medium/Bodies/M_MED_Quill/Meshes/RavenQuill_02_Rigids/M_MED_Quill_02_AnimBp
target_skeleton=&/M_MED_Quill_02_Skeleton
graphs [
 AnimGraph {nodes [
   Anim Dynamics - Bone: dyn_armband_knot(Alpha=0.75, ExternalForce="0,0,0") {
    alpha=0.9
    angularDampingOverride=0.999
    angularSpringConstant=1500
    bAngularSpring=true
    bGravityOverrideInSimSpace=true
    bOverrideAngularDamping=true
    simulationSpace=Actor
    alphaBoolBlend {blendOption=Linear}
    boundBone {boneName=dyn_armband_knot}
    physicsBodyDefinitions [{
      sphereCollisionRadius=0
      boundBone {boneName=dyn_armband_knot}
      boxExtents=<15 5 5>
      constraintSetup {angularTargetAxis=AxisZ, twistAxis=AxisY, angularLimitsMax=<15 0 15>, angularLimitsMin=<-15 0 -15>, angularTarget=<0 0 1>}
      localJointOffset=<0 0 0>
     }]
   }
   n1 = Control Rig(Source=Copy Pose From Mesh(SourceMeshComponent=Get MeshToCopy() -> (MeshToCopy:Skeletal Mesh Component Object Reference))) {
    alphaBoolBlend {blendOption=Linear}
    controlRigAssetReference {blueprintRigClass=&/M_MED_Quill_02_CtrlRig.M_MED_Quill_02_CtrlRig_C}
   }
   n2 = Anim Dynamics - Bone: dyn_shoulder_l(ComponentPose=Local To Component(LocalPose=@n1), Alpha=0.5, ExternalForce="0,0,0") {
    angularDampingOverride=0.9999
    bAngularSpring=true
    bGravityOverrideInSimSpace=true
    bOverrideAngularDamping=true
    bUseGravityOverride=true
    lODThreshold=1
    simulationSpace=RootRelative
    alphaBoolBlend {blendOption=Linear}
    boundBone {boneName=dyn_shoulder_l}
    gravityOverride {z=-980}
    physicsBodyDefinitions [{
      sphereCollisionRadius=0
      boundBone {boneName=dyn_shoulder_l}
      boxExtents=<10 20 10>
      constraintSetup {angularLimitsMax=<15 10 0>, angularLimitsMin=<-15 -10 0>, angularTarget=<0 0 0>, linearAxesMin=<-5 0 0>}
      localJointOffset=<10 0 -5>
     }]
   }
   n3 = Anim Dynamics - Bone: dyn_shoulder_r(ComponentPose=@n2, Alpha=0.5, ExternalForce="0,0,0") {
    angularDampingOverride=0.9999
    bAngularSpring=true
    bGravityOverrideInSimSpace=true
    bOverrideAngularDamping=true
    bUseGravityOverride=true
    lODThreshold=1
    simulationSpace=RootRelative
    alphaBoolBlend {blendOption=Linear}
    boundBone {boneName=dyn_shoulder_r}
    gravityOverride {z=-980}
    physicsBodyDefinitions [{
      sphereCollisionRadius=0
      boundBone {boneName=dyn_shoulder_r}
      boxExtents=<10 20 10>
      constraintSetup {angularLimitsMax=<15 10 0>, angularLimitsMin=<-15 -10 0>, angularTarget=<0 0 0>, linearAxesMin=<-5 0 0>}
      localJointOffset=<-10 0 3.5>
     }]
   }
   n4 = Anim Dynamics - Bone: Collar_R_02(ComponentPose=@n3, Alpha=1, ExternalForce="0,0,0") {
    angularDampingOverride=0.9996
    angularSpringConstant=6051.4448
    bAngularSpring=true
    bGravityOverrideInSimSpace=true
    bOverrideAngularDamping=true
    bUseGravityOverride=true
    simulationSpace=RootRelative
    alphaBoolBlend {blendOption=Linear}
    boundBone {boneName=Collar_R_02}
    gravityOverride {z=-980}
    physicsBodyDefinitions [{
      boundBone {boneName=Collar_R_02}
      boxExtents=<11 10 10>
      constraintSetup {angularTargetAxis=AxisY, angularLimitsMax=<0 0 20>, angularLimitsMin=<0 0 -20>, angularTarget=<0 1 1>}
      localJointOffset=<-1 0 0>
     }]
   }
   n5 = Anim Dynamics - Bone: Collar_R_03(ComponentPose=@n4, Alpha=1, ExternalForce="0,0,0") {
    angularDampingOverride=0.99
    angularSpringConstant=3152.9021
    bAngularSpring=true
    bGravityOverrideInSimSpace=true
    bOverrideAngularDamping=true
    bUseGravityOverride=true
    simulationSpace=RootRelative
    alphaBoolBlend {blendOption=Linear}
    boundBone {boneName=Collar_R_03}
    gravityOverride {z=-980}
    physicsBodyDefinitions [{
      boundBone {boneName=Collar_R_03}
      boxExtents=<10 10 10>
      constraintSetup {angularTargetAxis=AxisY, angularLimitsMax=<0 0 20>, angularLimitsMin=<0 0 -20>, angularTarget=<0 1 1>}
      localJointOffset=<-1 0 0>
     }]
   }
   n6 = Anim Dynamics - Bone: Collar_R_04(ComponentPose=@n5, Alpha=1, ExternalForce="0,0,0") {
    angularDampingOverride=0.999
    angularSpringConstant=8283.5908
    bAngularSpring=true
    bGravityOverrideInSimSpace=true
    bOverrideAngularDamping=true
    bUseGravityOverride=true
    simulationSpace=RootRelative
    alphaBoolBlend {blendOption=Linear}
    boundBone {boneName=Collar_R_04}
    gravityOverride {z=-980}
    physicsBodyDefinitions [{
      boundBone {boneName=Collar_R_04}
      boxExtents=<12 10 10>
      constraintSetup {angularTargetAxis=AxisY, angularLimitsMax=<0 0 20>, angularLimitsMin=<0 0 -20>, angularTarget=<0 1 1>}
      localJointOffset=<-1 0 0>
     }]
   }
   n7 = Anim Dynamics - Bone: Collar_M_02(ComponentPose=@n6, Alpha=1, ExternalForce="0,0,0") {
    angularDampingOverride=1
    angularSpringConstant=6897.9834
    bAngularSpring=true
    bOverrideAngularDamping=true
    simulationSpace=RootRelative
    alphaBoolBlend {blendOption=Linear}
    boundBone {boneName=Collar_M_02}
    gravityOverride {z=-980}
    physicsBodyDefinitions [{
      boundBone {boneName=Collar_M_02}
      boxExtents=<13 10 10>
      constraintSetup {angularTargetAxis=AxisY, angularLimitsMax=<0 5 20>, angularLimitsMin=<0 -5 -20>, angularTarget=<0 1 1>}
      localJointOffset=<-1 0 0>
     }]
   }
   n8 = Anim Dynamics - Bone: Collar_M_03(ComponentPose=@n7, Alpha=1, ExternalForce="0,0,0") {
    angularDampingOverride=0.9999
    angularSpringConstant=3681.4236
    bAngularSpring=true
    bOverrideAngularDamping=true
    simulationSpace=RootRelative
    alphaBoolBlend {blendOption=Linear}
    boundBone {boneName=Collar_M_03}
    gravityOverride {z=-980}
    physicsBodyDefinitions [{
      boundBone {boneName=Collar_M_03}
      boxExtents=<13 10 10>
      constraintSetup {angularTargetAxis=AxisY, angularLimitsMax=<0 3 20>, angularLimitsMin=<0 -5 -20>, angularTarget=<0 1 1>}
      localJointOffset=<-1 0 0>
     }]
   }
   n9 = Anim Dynamics - Bone: Collar_M_04(ComponentPose=@n8, Alpha=1, ExternalForce="0,0,0") {
    angularDampingOverride=0.9999
    angularSpringConstant=2469.1172
    bAngularSpring=true
    bOverrideAngularDamping=true
    simulationSpace=RootRelative
    alphaBoolBlend {blendOption=Linear}
    boundBone {boneName=Collar_M_04}
    gravityOverride {z=-980}
    physicsBodyDefinitions [{
      boundBone {boneName=Collar_M_04}
      boxExtents=<13 10 10>
      constraintSetup {angularTargetAxis=AxisY, angularLimitsMax=<0 -8 20>, angularLimitsMin=<0 -8 -20>, angularTarget=<0 1 1>}
      localJointOffset=<-1 0 0>
     }]
   }
   n10 = Anim Dynamics - Bone: Collar_L_02(ComponentPose=@n9, Alpha=1, ExternalForce="0,0,0") {
    angularDampingOverride=0.9999
    angularSpringConstant=6052.5825
    bAngularSpring=true
    bOverrideAngularDamping=true
    simulationSpace=RootRelative
    alphaBoolBlend {blendOption=Linear}
    boundBone {boneName=Collar_L_02}
    gravityOverride {z=-980}
    physicsBodyDefinitions [{
      boundBone {boneName=Collar_L_02}
      boxExtents=<12 10 10>
      constraintSetup {angularTargetAxis=AxisY, angularLimitsMax=<0 0 20>, angularLimitsMin=<0 0 -20>, angularTarget=<0 1 1>}
      localJointOffset=<-1 0 0>
     }]
   }
   n11 = Anim Dynamics - Bone: Collar_L_03(ComponentPose=@n10, Alpha=1, ExternalForce="0,0,0") {
    angularDampingOverride=0.99
    angularSpringConstant=3682.8762
    bAngularSpring=true
    bOverrideAngularDamping=true
    simulationSpace=RootRelative
    alphaBoolBlend {blendOption=Linear}
    boundBone {boneName=Collar_L_03}
    gravityOverride {z=-980}
    physicsBodyDefinitions [{
      boundBone {boneName=Collar_L_03}
      boxExtents=<10 10 10>
      constraintSetup {angularTargetAxis=AxisY, angularLimitsMax=<0 0 20>, angularLimitsMin=<0 0 -20>, angularTarget=<0 1 1>}
      localJointOffset=<-1 0 0>
     }]
   }
   n12 = Anim Dynamics - Bone: Collar_L_04(ComponentPose=@n11, Alpha=1, ExternalForce="0,0,0") {
    angularDampingOverride=0.999
    angularSpringConstant=6484.3213
    bAngularSpring=true
    bOverrideAngularDamping=true
    simulationSpace=RootRelative
    alphaBoolBlend {blendOption=Linear}
    boundBone {boneName=Collar_L_04}
    gravityOverride {z=-980}
    physicsBodyDefinitions [{
      boundBone {boneName=Collar_L_04}
      boxExtents=<14 10 10>
      constraintSetup {angularTargetAxis=AxisY, angularLimitsMax=<0 0 20>, angularLimitsMin=<0 0 -20>, angularTarget=<0 1 1>}
      localJointOffset=<-1 0 0>
     }]
   }
   n13 = Anim Dynamics - Bone: dyn_armband_low(ComponentPose=@n12, Alpha=1, ExternalForce="0,0,0") {
    alpha=0.9
    angularDampingOverride=0.999
    angularSpringConstant=1964.7235
    bAngularSpring=true
    bGravityOverrideInSimSpace=true
    bOverrideAngularDamping=true
    bUseGravityOverride=true
    simulationSpace=RootRelative
    alphaBoolBlend {blendOption=Linear}
    boundBone {boneName=dyn_armband_low}
    gravityOverride {z=-980}
    physicsBodyDefinitions [{
      sphereCollisionRadius=0
      boundBone {boneName=dyn_armband_low}
      boxExtents=<15 5 5>
      constraintSetup {angularTargetAxis=AxisZ, twistAxis=AxisY, angularLimitsMax=<15 15 15>, angularLimitsMin=<-15 -15 -15>, angularTarget=<0 0 1>}
      localJointOffset=<0 0 0.5>
     }]
   }
   n14 = Anim Dynamics - Bone: dyn_armband_hi(ComponentPose=@n13, Alpha=1, ExternalForce="0,0,0") {
    alpha=0.9
    angularDampingOverride=0.999
    angularSpringConstant=2386.4805
    bAngularSpring=true
    bGravityOverrideInSimSpace=true
    bOverrideAngularDamping=true
    bUseGravityOverride=true
    simulationSpace=RootRelative
    alphaBoolBlend {blendOption=Linear}
    boundBone {boneName=dyn_armband_hi}
    gravityOverride {z=-980}
    physicsBodyDefinitions [{
      sphereCollisionRadius=0
      boundBone {boneName=dyn_armband_hi}
      boxExtents=<15 5 5>
      constraintSetup {angularTargetAxis=AxisZ, twistAxis=AxisY, angularLimitsMax=<15 15 15>, angularLimitsMin=<-15 -15 -15>, angularTarget=<0 0 1>}
      localJointOffset=<0 0 0.5>
     }]
   }
   n15 = Anim Dynamics - Bone: dyn_skirtFeathers(ComponentPose=@n14, Alpha=0.5, ExternalForce="0,0,0") {
    bGravityOverrideInSimSpace=true
    gravityScale=0.15
    lODThreshold=1
    numSolverIterationsPostUpdate=4
    numSolverIterationsPreUpdate=16
    simulationSpace=Actor
    alphaBoolBlend {blendOption=Linear}
    boundBone {boneName=dyn_skirtFeathers}
    physicsBodyDefinitions [{
      sphereCollisionRadius=0
      boundBone {boneName=dyn_skirtFeathers}
      boxExtents=<20 20 1>
      constraintSetup {angularTargetAxis=AxisY, twistAxis=AxisY, angularLimitsMax=<0 30 30>, angularLimitsMin=<0 -30 -30>, angularTarget=<0 0 0>}
      localJointOffset=<0 -1 0>
     }]
   }
   n16 = RigidBody(ComponentPose=@n15) {
    bClampLinearTranslationLimitToRefPose=true
    bOverrideWorldGravity=true
    overridePhysicsAsset=&/M_MED_Quill_02_Physics
    simulationSpace=BaseBoneSpace
    alphaBoolBlend {blendOption=Linear}
    baseBoneRef {boneName=root}
    componentAppliedLinearAccClamp=<3000 3000 3000>
    componentLinearAccScale=<0.25 0.25 0.25>
    componentLinearVelScale=<0.25 0.25 0.25>
    overrideWorldGravity {z=-980}
   }
   Output Pose(Result=Component To Local(ComponentPose=@n16)) {layerGroup=None}
  ]}
 EventGraph {nodes [
   Event BlueprintUpdateAnimation()
   > Set Disabel_AnimDynamics(Disabel_AnimDynamics:float=GetCurveValue(CurveName=UpperArm_R_Up)) -> (Output_Get:float)
   Event BlueprintInitializeAnimation()
   > Set MeshToCopy(MeshToCopy:Skeletal Mesh Component Object Reference=GetOwnerSkeletalMeshComponent()) -> (Output_Get:Skeletal Mesh Component Object Reference)
   TryGetPawnOwner()
  ]}
]
variables [
 Disabel_AnimDynamics {type=number}
 MeshToCopy {type=SkeletalMeshComponent}
]
